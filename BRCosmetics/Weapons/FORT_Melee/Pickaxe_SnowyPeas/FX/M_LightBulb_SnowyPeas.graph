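# Material for light bulb effects on the SnowyPeas pickaxe, set up for Niagara mesh particles with Default Lit shading.
# Its graph holds a HueShift function whose percentage comes from the HueShift channel of a Dynamic Parameter, exposing it to per-particle control.
# No color input to that function and no connection from it to a material output appears in the graph.
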Material /BRCosmetics/Weapons/FORT_Melee/Pickaxe_SnowyPeas/FX/M_LightBulb_SnowyPeas
graph {
 properties {
  bAutomaticallySetUsageInEditor=false
  bUsedWithNiagaraMeshParticles=true
  previewMesh=/Game/Effects/Fort_Effects/Mesh/SM_Comet_Center_02
  usedShadingModels=Default Lit
 }
 nodes [MaterialFunctionCall(Hue Shift Percentage=DynamicParameter() {paramNames [HueShift,Param2,Param3,Param4]}.HueShift) {materialFunction=&HueShift}]
}
